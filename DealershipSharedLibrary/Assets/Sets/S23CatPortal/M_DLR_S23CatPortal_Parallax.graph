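# Particle-driven material for the S23 cat portal set, layering panned textures into a starry, distorted portal look.
# Each layer can blend between alternate coordinates, including camera-vector UVs for fake parallax depth.
# Dynamic-parameter sphere masks and radial-UV noise shape the portal opacity and outline.
# A cat variant adds world-position dissolve and Fresnel glow, with scalar blends selecting modes, optional temporal dither, and shared emissive adjustments.
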Material /DealershipSharedLibrary/Assets/Sets/S23CatPortal/M_DLR_S23CatPortal_Parallax
graph {
 expressions {
  MaterialExpressionAdd_0 {
   class=MaterialExpressionAdd
   inputs [
    A {from=MaterialExpressionMultiply_1}
    B {from=MaterialExpressionConvert_0}
   ]
  }
  MaterialExpressionAdd_10 {
   class=MaterialExpressionAdd
   inputs [
    A {from=MaterialExpressionMultiply_16}
    B {from=MaterialExpressionConvert_6}
   ]
  }
  MaterialExpressionAdd_11 {
   class=MaterialExpressionAdd
   inputs [
    A {from=MaterialExpressionAdd_9}
    B {from=MaterialExpressionNamedRerouteUsage_22}
   ]
  }
  MaterialExpressionAdd_12 {
   class=MaterialExpressionAdd
   inputs [
    A {from=MaterialExpressionSaturate_3}
    B {from=MaterialExpressionReroute_5}
   ]
  }
  MaterialExpressionAdd_13 {
   class=MaterialExpressionAdd
   inputs [
    A {from=MaterialExpressionAdd_12}
    B {from=MaterialExpressionSaturate_4}
   ]
  }
  MaterialExpressionAdd_15 {
   class=MaterialExpressionAdd
   inputs [
    A {from=MaterialExpressionChannelMaskParameter_6, output=R}
    B {from=MaterialExpressionScalarParameter_18}
   ]
  }
  MaterialExpressionAdd_16 {
   class=MaterialExpressionAdd
   inputs [
    A {from=MaterialExpressionMultiply_26}
    B {from=MaterialExpressionAdd_17}
   ]
  }
  MaterialExpressionAdd_17 {
   class=MaterialExpressionAdd
   inputs [
    A {from=MaterialExpressionNamedRerouteUsage_33}
    B {from=MaterialExpressionNamedRerouteUsage_34}
   ]
  }
  MaterialExpressionAdd_3 {
   class=MaterialExpressionAdd
   inputs [
    A {from=MaterialExpressionMultiply_4}
    B {from=MaterialExpressionConvert_2}
   ]
  }
  MaterialExpressionAdd_4 {
   class=MaterialExpressionAdd
   inputs [
    A {from=MaterialExpressionMultiply_5}
    B {from=MaterialExpressionConvert_3}
   ]
  }
  MaterialExpressionAdd_5 {
   class=MaterialExpressionAdd
   inputs [
    A {from=MaterialExpressionMultiply_6}
    B {from=MaterialExpressionConvert_4}
   ]
  }
  MaterialExpressionAdd_6 {
   class=MaterialExpressionAdd
   inputs [
    A {from=MaterialExpressionReroute_2}
    B {from=MaterialExpressionPower_1}
   ]
  }
  MaterialExpressionAdd_7 {
   class=MaterialExpressionAdd
   inputs [
    A {from=MaterialExpressionAdd_6}
    B {from=MaterialExpressionNamedRerouteUsage_28}
   ]
  }
  MaterialExpressionAdd_8 {
   class=MaterialExpressionAdd
   inputs [
    A {from=MaterialExpressionMultiply_8}
    B {from=MaterialExpressionScalarParameter_5}
   ]
  }
  MaterialExpressionAdd_9 {
   class=MaterialExpressionAdd
   inputs [
    A {from=MaterialExpressionMultiply_15}
    B {from=MaterialExpressionConvert_5}
   ]
  }
  MaterialExpressionCameraVectorWS_1 {class=MaterialExpressionCameraVectorWS, inputs [], properties {}}
  MaterialExpressionChannelMaskParameter_0 {
   class=MaterialExpressionChannelMaskParameter
   inputs [None {from=MaterialExpressionTextureSampleParameter2D_0, output=RGBA}]
   properties {group=01_BaseTexture, parameterName=Base Texture Channel Mask}
  }
  MaterialExpressionChannelMaskParameter_1 {
   class=MaterialExpressionChannelMaskParameter
   inputs [None {from=MaterialExpressionTextureSampleParameter2D_1, output=RGBA}]
   properties {group=06_Radial Opacity, parameterName=Radial Uvs Alpha Channel Mask}
  }
  MaterialExpressionChannelMaskParameter_2 {
   class=MaterialExpressionChannelMaskParameter
   inputs [None {from=MaterialExpressionTextureSampleParameter2D_4, output=RGBA}]
   properties {group=05_Distorted Texture, parameterName=Distorted Texture Channel Mask}
  }
  MaterialExpressionChannelMaskParameter_3 {
   class=MaterialExpressionChannelMaskParameter
   inputs [None {from=MaterialExpressionTextureSampleParameter2D_3, output=RGBA}]
   properties {group=03_Third Texture, maskChannel=Green, parameterName=Third Texture Channel Mask, defaultValue=<rgba 0 1 0 0>}
  }
  MaterialExpressionChannelMaskParameter_4 {
   class=MaterialExpressionChannelMaskParameter
   inputs [None {from=MaterialExpressionTextureSampleParameter2D_2, output=RGBA}]
   properties {group=02_Secondary Texture, maskChannel=Green, parameterName=Second Texture Channel Mask, defaultValue=<rgba 0 1 0 0>}
  }
  MaterialExpressionChannelMaskParameter_5 {
   class=MaterialExpressionChannelMaskParameter
   inputs [None {from=MaterialExpressionTextureSampleParameter2D_5, output=RGBA}]
   properties {group=04_Distortion Texture, parameterName="Distortion  Texture Channel Mask"}
  }
  MaterialExpressionChannelMaskParameter_6 {
   class=MaterialExpressionChannelMaskParameter
   inputs [None {from=MaterialExpressionIf_1, output=""}]
   properties {group=07_WorldPosition Dissolve, maskChannel=Green, parameterName=World Position Mask, defaultValue=<rgba 0 1 0 0>}
  }
  MaterialExpressionChannelMaskParameter_7 {
   class=MaterialExpressionChannelMaskParameter
   inputs [None {from=MaterialExpressionWorldPosition_1, output=XYZ}]
   properties {group=07_WorldPosition Dissolve, maskChannel=Green, parameterName=World Position Mask, defaultValue=<rgba 0 1 0 0>}
  }
  MaterialExpressionComponentMask_0 {class=MaterialExpressionComponentMask, inputs [None {from=MaterialExpressionVectorParameter_0, output=RGB}], properties {g=true, r=true}}
  MaterialExpressionComponentMask_10 {class=MaterialExpressionComponentMask, inputs [None {from=MaterialExpressionVectorParameter_7, output=RGB}], properties {g=true, r=true}}
  MaterialExpressionComponentMask_11 {class=MaterialExpressionComponentMask, inputs [None {from=MaterialExpressionVectorParameter_2, output=RGB}], properties {g=true, r=true}}
  MaterialExpressionComponentMask_12 {class=MaterialExpressionComponentMask, inputs [None {from=MaterialExpressionVectorParameter_8, output=RGB}], properties {g=true, r=true}}
  MaterialExpressionComponentMask_13 {class=MaterialExpressionComponentMask, inputs [None {from=MaterialExpressionVectorParameter_9, output=RGB}], properties {g=true, r=true}}
  MaterialExpressionComponentMask_14 {class=MaterialExpressionComponentMask, inputs [None {from=MaterialExpressionVectorParameter_11, output=RGB}], properties {g=true, r=true}}
  MaterialExpressionComponentMask_16 {
   class=MaterialExpressionComponentMask
   inputs [None {from=MaterialExpressionScreenPosition_0, output=ViewportUV}]
   properties {g=true, r=true}
  }
  MaterialExpressionComponentMask_3 {class=MaterialExpressionComponentMask, inputs [None {from=MaterialExpressionTransform_1}], properties {g=true, r=true}}
  MaterialExpressionComponentMask_4 {class=MaterialExpressionComponentMask, inputs [None {from=MaterialExpressionVectorParameter_3, output=RGB}], properties {g=true, r=true}}
  MaterialExpressionComponentMask_5 {class=MaterialExpressionComponentMask, inputs [None {from=MaterialExpressionVectorParameter_4, output=RGB}], properties {g=true, r=true}}
  MaterialExpressionComponentMask_6 {class=MaterialExpressionComponentMask, inputs [None {from=MaterialExpressionVectorParameter_5, output=RGB}], properties {g=true, r=true}}
  MaterialExpressionComponentMask_7 {class=MaterialExpressionComponentMask, inputs [None {from=MaterialExpressionVectorParameter_6, output=RGB}], properties {g=true, r=true}}
  MaterialExpressionComponentMask_8 {class=MaterialExpressionComponentMask, inputs [None {from=MaterialExpressionPanner_5}], properties {r=true}}
  MaterialExpressionComponentMask_9 {class=MaterialExpressionComponentMask, inputs [None {from=MaterialExpressionPanner_5}], properties {g=true}}
  MaterialExpressionConstant_0 {class=MaterialExpressionConstant, properties {r=0.5}}
  MaterialExpressionConstant_1 {class=MaterialExpressionConstant, properties {r=0.5}}
  MaterialExpressionConstant_2 {class=MaterialExpressionConstant, properties {r=1}}
  MaterialExpressionConvert_0 {
   class=MaterialExpressionConvert
   inputs [
    None {from=MaterialExpressionVectorParameter_0, output=Offset X}
    None {from=MaterialExpressionVectorParameter_0, output=Offset X}
   ]
   properties {
    nodeName=Make Vector2
    convertInputs [
     {
      type=Scalar
      expressionInput {
       expression=&self:MaterialExpressionVectorParameter_0
       inputName=None
       mask=1
       maskA=0
       maskB=1
       maskG=0
       maskR=0
       outputIndex=3
      }
     }
     {
      type=Scalar
      expressionInput {
       expression=&self:MaterialExpressionVectorParameter_0
       inputName=None
       mask=1
       maskA=1
       maskB=0
       maskG=0
       maskR=0
       outputIndex=4
      }
     }
    ]
    convertOutputs [{type=Vector2}]
   }
  }
  MaterialExpressionConvert_2 {
   class=MaterialExpressionConvert
   inputs [
    None {from=MaterialExpressionVectorParameter_3, output=Offset X}
    None {from=MaterialExpressionVectorParameter_3, output=Offset X}
   ]
   properties {
    nodeName=Make Vector2
    convertInputs [
     {
      type=Scalar
      expressionInput {
       expression=&self:MaterialExpressionVectorParameter_3
       inputName=None
       mask=1
       maskA=0
       maskB=1
       maskG=0
       maskR=0
       outputIndex=3
      }
     }
     {
      type=Scalar
      expressionInput {
       expression=&self:MaterialExpressionVectorParameter_3
       inputName=None
       mask=1
       maskA=1
       maskB=0
       maskG=0
       maskR=0
       outputIndex=4
      }
     }
    ]
    convertOutputs [{type=Vector2}]
   }
  }
  MaterialExpressionConvert_3 {
   class=MaterialExpressionConvert
   inputs [
    None {from=MaterialExpressionVectorParameter_4, output=Offset X}
    None {from=MaterialExpressionVectorParameter_4, output=Offset X}
   ]
   properties {
    nodeName=Make Vector2
    convertInputs [
     {
      type=Scalar
      expressionInput {
       expression=&self:MaterialExpressionVectorParameter_4
       inputName=None
       mask=1
       maskA=0
       maskB=1
       maskG=0
       maskR=0
       outputIndex=3
      }
     }
     {
      type=Scalar
      expressionInput {
       expression=&self:MaterialExpressionVectorParameter_4
       inputName=None
       mask=1
       maskA=1
       maskB=0
       maskG=0
       maskR=0
       outputIndex=4
      }
     }
    ]
    convertOutputs [{type=Vector2}]
   }
  }
  MaterialExpressionConvert_4 {
   class=MaterialExpressionConvert
   inputs [
    None {from=MaterialExpressionVectorParameter_5, output=Offset X}
    None {from=MaterialExpressionVectorParameter_5, output=Offset X}
   ]
   properties {
    nodeName=Make Vector2
    convertInputs [
     {
      type=Scalar
      expressionInput {
       expression=&self:MaterialExpressionVectorParameter_5
       inputName=None
       mask=1
       maskA=0
       maskB=1
       maskG=0
       maskR=0
       outputIndex=3
      }
     }
     {
      type=Scalar
      expressionInput {
       expression=&self:MaterialExpressionVectorParameter_5
       inputName=None
       mask=1
       maskA=1
       maskB=0
       maskG=0
       maskR=0
       outputIndex=4
      }
     }
    ]
    convertOutputs [{type=Vector2}]
   }
  }
  MaterialExpressionConvert_5 {
   class=MaterialExpressionConvert
   inputs [
    None {from=MaterialExpressionVectorParameter_7, output=Offset X}
    None {from=MaterialExpressionVectorParameter_7, output=Offset X}
   ]
   properties {
    nodeName=Make Vector2
    convertInputs [
     {
      type=Scalar
      expressionInput {
       expression=&self:MaterialExpressionVectorParameter_7
       inputName=None
       mask=1
       maskA=0
       maskB=1
       maskG=0
       maskR=0
       outputIndex=3
      }
     }
     {
      type=Scalar
      expressionInput {
       expression=&self:MaterialExpressionVectorParameter_7
       inputName=None
       mask=1
       maskA=1
       maskB=0
       maskG=0
       maskR=0
       outputIndex=4
      }
     }
    ]
    convertOutputs [{type=Vector2}]
   }
  }
  MaterialExpressionConvert_6 {
   class=MaterialExpressionConvert
   inputs [
    None {from=MaterialExpressionVectorParameter_2, output=Offset X}
    None {from=MaterialExpressionVectorParameter_2, output=Offset X}
   ]
   properties {
    nodeName=Make Vector2
    convertInputs [
     {
      type=Scalar
      expressionInput {
       expression=&self:MaterialExpressionVectorParameter_2
       inputName=None
       mask=1
       maskA=0
       maskB=1
       maskG=0
       maskR=0
       outputIndex=3
      }
     }
     {
      type=Scalar
      expressionInput {
       expression=&self:MaterialExpressionVectorParameter_2
       inputName=None
       mask=1
       maskA=1
       maskB=0
       maskG=0
       maskR=0
       outputIndex=4
      }
     }
    ]
    convertOutputs [{type=Vector2}]
   }
  }
  MaterialExpressionConvert_7 {
   class=MaterialExpressionConvert
   inputs [
    None {from=MaterialExpressionVectorParameter_8, output=Speed X Second Texture}
    None {from=MaterialExpressionVectorParameter_8, output=Speed X Second Texture}
   ]
   properties {
    nodeName=Make Vector2
    convertInputs [
     {
      type=Scalar
      expressionInput {
       expression=&self:MaterialExpressionVectorParameter_8
       inputName=None
       mask=1
       maskA=0
       maskB=1
       maskG=0
       maskR=0
       outputIndex=3
      }
     }
     {
      type=Scalar
      expressionInput {
       expression=&self:MaterialExpressionVectorParameter_8
       inputName=None
       mask=1
       maskA=1
       maskB=0
       maskG=0
       maskR=0
       outputIndex=4
      }
     }
    ]
    convertOutputs [{type=Vector2}]
   }
  }
  MaterialExpressionConvert_8 {
   class=MaterialExpressionConvert
   inputs [
    None {from=MaterialExpressionVectorParameter_9, output=Speed X Distorted Texture}
    None {from=MaterialExpressionVectorParameter_9, output=Speed X Distorted Texture}
   ]
   properties {
    nodeName=Make Vector2
    convertInputs [
     {
      type=Scalar
      expressionInput {
       expression=&self:MaterialExpressionVectorParameter_9
       inputName=None
       mask=1
       maskA=0
       maskB=1
       maskG=0
       maskR=0
       outputIndex=3
      }
     }
     {
      type=Scalar
      expressionInput {
       expression=&self:MaterialExpressionVectorParameter_9
       inputName=None
       mask=1
       maskA=1
       maskB=0
       maskG=0
       maskR=0
       outputIndex=4
      }
     }
    ]
    convertOutputs [{type=Vector2}]
   }
  }
  MaterialExpressionConvert_9 {
   class=MaterialExpressionConvert
   inputs [
    None {from=MaterialExpressionVectorParameter_6, output=VectorToRadial Speed X}
    None {from=MaterialExpressionVectorParameter_6, output=VectorToRadial Speed X}
   ]
   properties {
    nodeName=Make Vector2
    convertInputs [
     {
      type=Scalar
      expressionInput {
       expression=&self:MaterialExpressionVectorParameter_6
       inputName=None
       mask=1
       maskA=0
       maskB=1
       maskG=0
       maskR=0
       outputIndex=3
      }
     }
     {
      type=Scalar
      expressionInput {
       expression=&self:MaterialExpressionVectorParameter_6
       inputName=None
       mask=1
       maskA=1
       maskB=0
       maskG=0
       maskR=0
       outputIndex=4
      }
     }
    ]
    convertOutputs [{type=Vector2}]
   }
  }
  MaterialExpressionDDX_0 {class=MaterialExpressionDDX, inputs [Value {from=MaterialExpressionComponentMask_8, output=""}], properties {}}
  MaterialExpressionDDY_0 {class=MaterialExpressionDDY, inputs [Value {from=MaterialExpressionComponentMask_9, output=""}], properties {}}
  MaterialExpressionDesaturation_0 {
   class=MaterialExpressionDesaturation
   inputs [
    None {from=MaterialExpressionAdd_13, output=""}
    Fraction {from=MaterialExpressionScalarParameter_9, output=""}
   ]
   properties {}
  }
  MaterialExpressionDesaturation_1 {
   class=MaterialExpressionDesaturation
   inputs [
    None {from=MaterialExpressionVectorParameter_10, output=RGB}
    Fraction {from=MaterialExpressionScalarParameter_10, output=""}
   ]
   properties {}
  }
  MaterialExpressionDynamicParameter_0 {
   class=MaterialExpressionDynamicParameter
   inputs []
   properties {paramNames [SphereMask Radius In Dissolve, SphereMask 2 Radius Out Dissolve, Param3, Param4]}
  }
  MaterialExpressionFresnel_0 {
   class=MaterialExpressionFresnel
   inputs [
    ExponentIn {from=MaterialExpressionScalarParameter_31, output=""}
    BaseReflectFractionIn {from="", output=""}
    Normal {from="", output=""}
   ]
   properties {exponent=3}
  }
  MaterialExpressionIf_0 {
   class=MaterialExpressionIf
   inputs [
    A {from=MaterialExpressionConstant_2, output=""}
    B {from=MaterialExpressionNamedRerouteUsage_27, output=""}
    "A > B" {from=MaterialExpressionNamedRerouteUsage_29, output=""}
    "A == B" {from=MaterialExpressionNamedRerouteUsage_30, output=""}
    "A < B" {from=MaterialExpressionNamedRerouteUsage_30, output=""}
   ]
   properties {}
  }
  MaterialExpressionIf_1 {
   class=MaterialExpressionIf
   inputs [
    A {from=MaterialExpressionChannelMaskParameter_7, output=R}
    B {from="", output=""}
    "A > B" {from=MaterialExpressionSubtract_4, output=""}
    "A == B" {from=MaterialExpressionSubtract_4, output=""}
    "A < B" {from=MaterialExpressionSubtract_3, output=""}
   ]
   properties {}
  }
  MaterialExpressionLinearInterpolate_0 {
   class=MaterialExpressionLinearInterpolate
   inputs [
    A {from=MaterialExpressionChannelMaskParameter_0, output=R}
    B {from=MaterialExpressionTextureSampleParameter2D_0, output=RGB}
    Alpha {from=MaterialExpressionScalarParameter_1}
   ]
  }
  MaterialExpressionLinearInterpolate_1 {
   class=MaterialExpressionLinearInterpolate
   inputs [
    A {from=MaterialExpressionSphereMask_1}
    B {from=MaterialExpressionOneMinus_0}
    Alpha {from=MaterialExpressionScalarParameter_4}
   ]
  }
  MaterialExpressionLinearInterpolate_2 {
   class=MaterialExpressionLinearInterpolate
   inputs [
    A {from=MaterialExpressionLinearInterpolate_9}
    B {from=MaterialExpressionMaterialFunctionCall_1, output=Result}
    Alpha {from=MaterialExpressionScalarParameter_6}
   ]
  }
  MaterialExpressionLinearInterpolate_3 {
   class=MaterialExpressionLinearInterpolate
   inputs [
    A {from=MaterialExpressionChannelMaskParameter_2, output=R}
    B {from=MaterialExpressionReroute_4}
    Alpha {from=MaterialExpressionScalarParameter_8}
   ]
  }
  MaterialExpressionLinearInterpolate_4 {
   class=MaterialExpressionLinearInterpolate
   inputs [
    A {from=MaterialExpressionNamedRerouteUsage_0}
    B {from=MaterialExpressionComponentMask_16}
    Alpha {from=MaterialExpressionScalarParameter_21}
   ]
  }
  MaterialExpressionLinearInterpolate_5 {
   class=MaterialExpressionLinearInterpolate
   inputs [
    A {from=MaterialExpressionNamedRerouteUsage_35}
    B {from=MaterialExpressionNamedRerouteUsage_36}
    Alpha {from=MaterialExpressionScalarParameter_22}
   ]
  }
  MaterialExpressionLinearInterpolate_6 {
   class=MaterialExpressionLinearInterpolate
   inputs [
    A {from=MaterialExpressionNamedRerouteUsage_37}
    B {from=MaterialExpressionNamedRerouteUsage_38}
    Alpha {from=MaterialExpressionScalarParameter_23}
   ]
  }
  MaterialExpressionLinearInterpolate_7 {
   class=MaterialExpressionLinearInterpolate
   inputs [
    A {from=MaterialExpressionNamedRerouteUsage_39}
    B {from=MaterialExpressionNamedRerouteUsage_40}
    Alpha {from=MaterialExpressionScalarParameter_24}
   ]
  }
  MaterialExpressionLinearInterpolate_8 {
   class=MaterialExpressionLinearInterpolate
   inputs [
    A {from=MaterialExpressionIf_0}
    B {from=MaterialExpressionMultiply_27}
    Alpha {from=MaterialExpressionScalarParameter_26}
   ]
  }
  MaterialExpressionLinearInterpolate_9 {
   class=MaterialExpressionLinearInterpolate
   inputs [
    A {from=MaterialExpressionNamedRerouteUsage_17}
    B {from=MaterialExpressionNamedRerouteUsage_32}
    Alpha {from=MaterialExpressionScalarParameter_30}
   ]
  }
  MaterialExpressionMaterialFunctionCall_0 {
   class=MaterialExpressionMaterialFunctionCall
   inputs [
    Vector or UVs {from=""}
    Swizzle Coordinate Output {from=""}
   ]
   properties {materialFunction=&VectorToRadialValue}
  }
  MaterialExpressionMaterialFunctionCall_1 {
   class=MaterialExpressionMaterialFunctionCall
   inputs [
    Alpha Threshold {from=MaterialExpressionLinearInterpolate_9}
    Random {from=""}
   ]
   properties {materialFunction=&DitherTemporalAA}
  }
  MaterialExpressionMaterialFunctionCall_2 {
   class=MaterialExpressionMaterialFunctionCall
   inputs [Emissive {from=MaterialExpressionLinearInterpolate_8}]
   properties {materialFunction=&MF_FX_EmissiveAdjustments}
  }
  MaterialExpressionMultiply_0 {
   class=MaterialExpressionMultiply
   inputs [
    A {from=MaterialExpressionPower_0}
    B {from=MaterialExpressionScalarParameter_27}
   ]
   properties {constB=10}
  }
  MaterialExpressionMultiply_1 {
   class=MaterialExpressionMultiply
   inputs [
    A {from=MaterialExpressionLinearInterpolate_4}
    B {from=MaterialExpressionComponentMask_0}
   ]
  }
  MaterialExpressionMultiply_10 {
   class=MaterialExpressionMultiply
   inputs [
    A {from=MaterialExpressionParticleColor_0, output=A}
    B {from=MaterialExpressionNamedRerouteDeclaration_15}
   ]
  }
  MaterialExpressionMultiply_11 {
   class=MaterialExpressionMultiply
   inputs [
    A {from=MaterialExpressionNamedRerouteUsage_15}
    B {from=MaterialExpressionNamedRerouteUsage_16}
   ]
  }
  MaterialExpressionMultiply_12 {
   class=MaterialExpressionMultiply
   inputs [
    A {from=MaterialExpressionMultiply_10}
    B {from=MaterialExpressionMultiply_11}
   ]
  }
  MaterialExpressionMultiply_13 {
   class=MaterialExpressionMultiply
   inputs [
    A {from=MaterialExpressionChannelMaskParameter_4, output=R}
    B {from=MaterialExpressionReroute_3}
   ]
  }
  MaterialExpressionMultiply_14 {
   class=MaterialExpressionMultiply
   inputs [
    A {from=MaterialExpressionMultiply_13}
    B {from=MaterialExpressionScalarParameter_28}
   ]
   properties {constB=2000}
  }
  MaterialExpressionMultiply_15 {
   class=MaterialExpressionMultiply
   inputs [
    A {from=MaterialExpressionLinearInterpolate_7}
    B {from=MaterialExpressionComponentMask_10}
   ]
  }
  MaterialExpressionMultiply_16 {
   class=MaterialExpressionMultiply
   inputs [
    A {from=MaterialExpressionNamedRerouteUsage_19}
    B {from=MaterialExpressionComponentMask_11}
   ]
  }
  MaterialExpressionMultiply_17 {
   class=MaterialExpressionMultiply
   inputs [
    A {from=MaterialExpressionChannelMaskParameter_5, output=R}
    B {from=MaterialExpressionScalarParameter_7}
   ]
  }
  MaterialExpressionMultiply_18 {
   class=MaterialExpressionMultiply
   inputs [
    A {from=MaterialExpressionSubtract_0}
    B {from=MaterialExpressionScalarParameter_29}
   ]
   properties {constB=10}
  }
  MaterialExpressionMultiply_19 {
   class=MaterialExpressionMultiply
   inputs [
    A {from=MaterialExpressionDesaturation_0}
    B {from=MaterialExpressionParticleColor_1, output=RGB}
   ]
  }
  MaterialExpressionMultiply_20 {
   class=MaterialExpressionMultiply
   inputs [
    A {from=MaterialExpressionDesaturation_1}
    B {from=MaterialExpressionScalarParameter_11}
   ]
  }
  MaterialExpressionMultiply_21 {
   class=MaterialExpressionMultiply
   inputs [
    A {from=MaterialExpressionParticleColor_1, output=RGB}
    B {from=MaterialExpressionMultiply_20}
   ]
  }
  MaterialExpressionMultiply_22 {
   class=MaterialExpressionMultiply
   inputs [
    A {from=MaterialExpressionPower_3}
    B {from=MaterialExpressionPower_4}
   ]
  }
  MaterialExpressionMultiply_23 {
   class=MaterialExpressionMultiply
   inputs [
    A {from=MaterialExpressionAdd_15}
    B {from=MaterialExpressionMultiply_22}
   ]
  }
  MaterialExpressionMultiply_24 {
   class=MaterialExpressionMultiply
   inputs [
    A {from=MaterialExpressionParticleColor_2, output=A}
    B {from=MaterialExpressionMultiply_23}
   ]
  }
  MaterialExpressionMultiply_25 {
   class=MaterialExpressionMultiply
   inputs [
    A {from=MaterialExpressionPower_5}
    B {from=MaterialExpressionVectorParameter_12, output=RGB}
   ]
  }
  MaterialExpressionMultiply_26 {
   class=MaterialExpressionMultiply
   inputs [
    A {from=MaterialExpressionMultiply_25}
    B {from=MaterialExpressionScalarParameter_20}
   ]
  }
  MaterialExpressionMultiply_27 {
   class=MaterialExpressionMultiply
   inputs [
    A {from=MaterialExpressionAdd_16}
    B {from=MaterialExpressionScalarParameter_25}
   ]
  }
  MaterialExpressionMultiply_4 {
   class=MaterialExpressionMultiply
   inputs [
    A {from=MaterialExpressionLinearInterpolate_5}
    B {from=MaterialExpressionComponentMask_4}
   ]
  }
  MaterialExpressionMultiply_5 {
   class=MaterialExpressionMultiply
   inputs [
    A {from=MaterialExpressionLinearInterpolate_6}
    B {from=MaterialExpressionComponentMask_5}
   ]
  }
  MaterialExpressionMultiply_6 {
   class=MaterialExpressionMultiply
   inputs [
    A {from=MaterialExpressionMaterialFunctionCall_0, output=Radial Coordinates}
    B {from=MaterialExpressionComponentMask_6}
   ]
  }
  MaterialExpressionMultiply_7 {
   class=MaterialExpressionMultiply
   inputs [
    A {from=MaterialExpressionNamedRerouteUsage_10}
    B {from=MaterialExpressionNamedRerouteUsage_9}
   ]
  }
  MaterialExpressionMultiply_8 {
   class=MaterialExpressionMultiply
   inputs [
    A {from=MaterialExpressionNamedRerouteUsage_12}
    B {from=MaterialExpressionNamedRerouteUsage_11}
   ]
  }
  MaterialExpressionMultiply_9 {
   class=MaterialExpressionMultiply
   inputs [
    A {from=MaterialExpressionPower_2}
    B {from=MaterialExpressionNamedRerouteUsage_14}
   ]
  }
  MaterialExpressionNamedRerouteDeclaration_0 {
   class=MaterialExpressionNamedRerouteDeclaration
   inputs [None {from=MaterialExpressionTextureCoordinate_1}]
   properties {name=UVs, nodeColor=<rgba 0.6172 1 0 1>}
  }
  MaterialExpressionNamedRerouteDeclaration_1 {
   class=MaterialExpressionNamedRerouteDeclaration
   inputs [None {from=MaterialExpressionDynamicParameter_0, output=SphereMask Radius In Dissolve}]
   properties {name=DP SphereMask Radius, nodeColor=<rgba 1 0 0.0058 1>}
  }
  MaterialExpressionNamedRerouteDeclaration_10 {
   class=MaterialExpressionNamedRerouteDeclaration
   inputs [None {from=MaterialExpressionConvert_8}]
   properties {name=Spped Panner distorted Texture, nodeColor=<rgba 1 0.4922 0 1>}
  }
  MaterialExpressionNamedRerouteDeclaration_11 {
   class=MaterialExpressionNamedRerouteDeclaration
   inputs [None {from=MaterialExpressionMultiply_17}]
   properties {name=Distortion, nodeColor=<rgba 0.1484 1 0 1>}
  }
  MaterialExpressionNamedRerouteDeclaration_12 {
   class=MaterialExpressionNamedRerouteDeclaration
   inputs [None {from=MaterialExpressionComponentMask_3}]
   properties {name=CameraVectorUVs, nodeColor=<rgba 0.0312 0 1 1>}
  }
  MaterialExpressionNamedRerouteDeclaration_13 {
   class=MaterialExpressionNamedRerouteDeclaration
   inputs [None {from=MaterialExpressionLinearInterpolate_3}]
   properties {name=Distorted Tex, nodeColor=<rgba 1 0 0.6328 1>}
  }
  MaterialExpressionNamedRerouteDeclaration_14 {
   class=MaterialExpressionNamedRerouteDeclaration
   inputs [None {from=MaterialExpressionMultiply_0}]
   properties {name=Main Tex, nodeColor=<rgba 0 0.9531 1 1>}
  }
  MaterialExpressionNamedRerouteDeclaration_15 {
   class=MaterialExpressionNamedRerouteDeclaration
   inputs [None {from=MaterialExpressionSaturate_0}]
   properties {name=Opacity 1, nodeColor=<rgba 1 0.7266 0 1>}
  }
  MaterialExpressionNamedRerouteDeclaration_16 {
   class=MaterialExpressionNamedRerouteDeclaration
   inputs [None {from=MaterialExpressionConvert_9}]
   properties {name=SpeedPanner Vector, nodeColor=<rgba 0.3828 0 1 1>}
  }
  MaterialExpressionNamedRerouteDeclaration_18 {
   class=MaterialExpressionNamedRerouteDeclaration
   inputs [None {from=MaterialExpressionMultiply_19}]
   properties {name=Color, nodeColor=<rgba 0.875 0 1 1>}
  }
  MaterialExpressionNamedRerouteDeclaration_19 {
   class=MaterialExpressionNamedRerouteDeclaration
   inputs [None {from=MaterialExpressionMultiply_21}]
   properties {name=Outline, nodeColor=<rgba 0 1 0.5547 1>}
  }
  MaterialExpressionNamedRerouteDeclaration_2 {
   class=MaterialExpressionNamedRerouteDeclaration
   inputs [None {from=MaterialExpressionDynamicParameter_0, output=SphereMask 2 Radius Out Dissolve}]
   properties {name=DP SphereMask 2 Radius, nodeColor=<rgba 1 0 0.0058 1>}
  }
  MaterialExpressionNamedRerouteDeclaration_20 {
   class=MaterialExpressionNamedRerouteDeclaration
   inputs [None {from=MaterialExpressionComponentMask_14}]
   properties {name=CatDis, nodeColor=<rgba 1 0.2344 0 1>}
  }
  MaterialExpressionNamedRerouteDeclaration_21 {
   class=MaterialExpressionNamedRerouteDeclaration
   inputs [None {from=MaterialExpressionSaturate_5}]
   properties {name=Opacity Cat, nodeColor=<rgba 0 0.0859 1 1>}
  }
  MaterialExpressionNamedRerouteDeclaration_22 {
   class=MaterialExpressionNamedRerouteDeclaration
   inputs [None {from=MaterialExpressionMultiply_14}]
   properties {name=Stars, nodeColor=<rgba 0.4062 1 0 1>}
  }
  MaterialExpressionNamedRerouteDeclaration_3 {
   class=MaterialExpressionNamedRerouteDeclaration
   inputs [None {from=MaterialExpressionOneMinus_1}]
   properties {name=SphereMask Invert, nodeColor=<rgba 0 1 0.0156 1>}
  }
  MaterialExpressionNamedRerouteDeclaration_4 {
   class=MaterialExpressionNamedRerouteDeclaration
   inputs [None {from=MaterialExpressionSphereMask_0}]
   properties {name=SphereMask 1, nodeColor=<rgba 1 0 0.3047 1>}
  }
  MaterialExpressionNamedRerouteDeclaration_5 {
   class=MaterialExpressionNamedRerouteDeclaration
   inputs [None {from=MaterialExpressionChannelMaskParameter_1, output=R}]
   properties {name=Radial Text, nodeColor=<rgba 0 0.625 1 1>}
  }
  MaterialExpressionNamedRerouteDeclaration_6 {
   class=MaterialExpressionNamedRerouteDeclaration
   inputs [None {from=MaterialExpressionAdd_8}]
   properties {name=SphereMask * Radial, nodeColor=<rgba 0 1 0.8125 1>}
  }
  MaterialExpressionNamedRerouteDeclaration_7 {
   class=MaterialExpressionNamedRerouteDeclaration
   inputs [None {from=MaterialExpressionLinearInterpolate_1}]
   properties {name=SphereMask 2, nodeColor=<rgba 1 0.4922 0 1>}
  }
  MaterialExpressionNamedRerouteDeclaration_8 {
   class=MaterialExpressionNamedRerouteDeclaration
   inputs [None {from=MaterialExpressionSaturate_1}]
   properties {name=Opacity Portal, nodeColor=<rgba 0.1719 0 1 1>}
  }
  MaterialExpressionNamedRerouteDeclaration_9 {
   class=MaterialExpressionNamedRerouteDeclaration
   inputs [None {from=MaterialExpressionConvert_7}]
   properties {name=Spped Panner Second Texture, nodeColor=<rgba 1 0.4922 0 1>}
  }
  MaterialExpressionNamedRerouteUsage_0 {class=MaterialExpressionNamedRerouteUsage}
  MaterialExpressionNamedRerouteUsage_10 {class=MaterialExpressionNamedRerouteUsage}
  MaterialExpressionNamedRerouteUsage_11 {class=MaterialExpressionNamedRerouteUsage}
  MaterialExpressionNamedRerouteUsage_12 {class=MaterialExpressionNamedRerouteUsage}
  MaterialExpressionNamedRerouteUsage_14 {class=MaterialExpressionNamedRerouteUsage}
  MaterialExpressionNamedRerouteUsage_15 {class=MaterialExpressionNamedRerouteUsage}
  MaterialExpressionNamedRerouteUsage_16 {class=MaterialExpressionNamedRerouteUsage}
  MaterialExpressionNamedRerouteUsage_17 {class=MaterialExpressionNamedRerouteUsage}
  MaterialExpressionNamedRerouteUsage_19 {class=MaterialExpressionNamedRerouteUsage}
  MaterialExpressionNamedRerouteUsage_20 {class=MaterialExpressionNamedRerouteUsage}
  MaterialExpressionNamedRerouteUsage_21 {class=MaterialExpressionNamedRerouteUsage}
  MaterialExpressionNamedRerouteUsage_22 {class=MaterialExpressionNamedRerouteUsage}
  MaterialExpressionNamedRerouteUsage_23 {class=MaterialExpressionNamedRerouteUsage}
  MaterialExpressionNamedRerouteUsage_24 {class=MaterialExpressionNamedRerouteUsage}
  MaterialExpressionNamedRerouteUsage_25 {class=MaterialExpressionNamedRerouteUsage}
  MaterialExpressionNamedRerouteUsage_26 {class=MaterialExpressionNamedRerouteUsage}
  MaterialExpressionNamedRerouteUsage_27 {class=MaterialExpressionNamedRerouteUsage}
  MaterialExpressionNamedRerouteUsage_28 {class=MaterialExpressionNamedRerouteUsage}
  MaterialExpressionNamedRerouteUsage_29 {class=MaterialExpressionNamedRerouteUsage}
  MaterialExpressionNamedRerouteUsage_30 {class=MaterialExpressionNamedRerouteUsage}
  MaterialExpressionNamedRerouteUsage_31 {class=MaterialExpressionNamedRerouteUsage}
  MaterialExpressionNamedRerouteUsage_32 {class=MaterialExpressionNamedRerouteUsage}
  MaterialExpressionNamedRerouteUsage_33 {class=MaterialExpressionNamedRerouteUsage}
  MaterialExpressionNamedRerouteUsage_34 {class=MaterialExpressionNamedRerouteUsage}
  MaterialExpressionNamedRerouteUsage_35 {class=MaterialExpressionNamedRerouteUsage}
  MaterialExpressionNamedRerouteUsage_36 {class=MaterialExpressionNamedRerouteUsage}
  MaterialExpressionNamedRerouteUsage_37 {class=MaterialExpressionNamedRerouteUsage}
  MaterialExpressionNamedRerouteUsage_38 {class=MaterialExpressionNamedRerouteUsage}
  MaterialExpressionNamedRerouteUsage_39 {class=MaterialExpressionNamedRerouteUsage}
  MaterialExpressionNamedRerouteUsage_4 {class=MaterialExpressionNamedRerouteUsage}
  MaterialExpressionNamedRerouteUsage_40 {class=MaterialExpressionNamedRerouteUsage}
  MaterialExpressionNamedRerouteUsage_5 {class=MaterialExpressionNamedRerouteUsage}
  MaterialExpressionNamedRerouteUsage_6 {class=MaterialExpressionNamedRerouteUsage}
  MaterialExpressionNamedRerouteUsage_7 {class=MaterialExpressionNamedRerouteUsage}
  MaterialExpressionNamedRerouteUsage_8 {class=MaterialExpressionNamedRerouteUsage}
  MaterialExpressionNamedRerouteUsage_9 {class=MaterialExpressionNamedRerouteUsage}
  MaterialExpressionObjectPositionWS_0 {class=MaterialExpressionObjectPositionWS, inputs [], properties {}}
  MaterialExpressionObjectPositionWS_1 {class=MaterialExpressionObjectPositionWS, inputs [], properties {}}
  MaterialExpressionOneMinus_0 {class=MaterialExpressionOneMinus, inputs [None {from=MaterialExpressionSphereMask_1}]}
  MaterialExpressionOneMinus_1 {class=MaterialExpressionOneMinus, inputs [None {from=MaterialExpressionSphereMask_0}]}
  MaterialExpressionOneMinus_2 {class=MaterialExpressionOneMinus, inputs [None {from=MaterialExpressionSphereMask_3}]}
  MaterialExpressionPanner_0 {
   class=MaterialExpressionPanner
   inputs [
    Coordinate {from=MaterialExpressionAdd_0}
    Time {from=""}
    Speed {from=MaterialExpressionComponentMask_12}
   ]
  }
  MaterialExpressionPanner_3 {
   class=MaterialExpressionPanner
   inputs [
    Coordinate {from=MaterialExpressionAdd_3}
    Time {from=""}
    Speed {from=MaterialExpressionNamedRerouteUsage_20}
   ]
  }
  MaterialExpressionPanner_4 {
   class=MaterialExpressionPanner
   inputs [
    Coordinate {from=MaterialExpressionAdd_4}
    Time {from=""}
    Speed {from=MaterialExpressionComponentMask_13}
   ]
  }
  MaterialExpressionPanner_5 {
   class=MaterialExpressionPanner
   inputs [
    Coordinate {from=MaterialExpressionAdd_5}
    Time {from=""}
    Speed {from=MaterialExpressionNamedRerouteUsage_7}
   ]
  }
  MaterialExpressionPanner_6 {
   class=MaterialExpressionPanner
   inputs [
    Coordinate {from=MaterialExpressionAdd_11}
    Time {from=""}
    Speed {from=MaterialExpressionNamedRerouteUsage_21}
   ]
  }
  MaterialExpressionPanner_7 {
   class=MaterialExpressionPanner
   inputs [
    Coordinate {from=MaterialExpressionAdd_10}
    Time {from=""}
    Speed {from=MaterialExpressionComponentMask_7}
   ]
  }
  MaterialExpressionParticleColor_0 {class=MaterialExpressionParticleColor, inputs [], properties {}}
  MaterialExpressionParticleColor_1 {class=MaterialExpressionParticleColor, inputs [], properties {}}
  MaterialExpressionParticleColor_2 {class=MaterialExpressionParticleColor, inputs [], properties {}}
  MaterialExpressionPower_0 {
   class=MaterialExpressionPower
   inputs [
    Base {from=MaterialExpressionLinearInterpolate_0}
    Exp {from=MaterialExpressionScalarParameter_0}
   ]
   properties {constExponent=1}
  }
  MaterialExpressionPower_1 {
   class=MaterialExpressionPower
   inputs [
    Base {from=MaterialExpressionMultiply_7}
    Exp {from=""}
   ]
   properties {constExponent=0.2}
  }
  MaterialExpressionPower_2 {
   class=MaterialExpressionPower
   inputs [
    Base {from=MaterialExpressionAdd_7}
    Exp {from=""}
   ]
   properties {constExponent=20}
  }
  MaterialExpressionPower_3 {
   class=MaterialExpressionPower
   inputs [
    Base {from=MaterialExpressionSphereMask_2}
    Exp {from=MaterialExpressionScalarParameter_14}
   ]
  }
  MaterialExpressionPower_4 {
   class=MaterialExpressionPower
   inputs [
    Base {from=MaterialExpressionOneMinus_2}
    Exp {from=MaterialExpressionScalarParameter_17}
   ]
  }
  MaterialExpressionPower_5 {
   class=MaterialExpressionPower
   inputs [
    Base {from=MaterialExpressionFresnel_0}
    Exp {from=MaterialExpressionScalarParameter_19}
   ]
  }
  MaterialExpressionReroute_1 {class=MaterialExpressionReroute, inputs [None {from=MaterialExpressionChannelMaskParameter_1, output=R}]}
  MaterialExpressionReroute_2 {class=MaterialExpressionReroute, inputs [None {from=MaterialExpressionMultiply_8}]}
  MaterialExpressionReroute_3 {class=MaterialExpressionReroute, inputs [None {from=MaterialExpressionChannelMaskParameter_3, output=R}]}
  MaterialExpressionReroute_4 {class=MaterialExpressionReroute, inputs [None {from=MaterialExpressionTextureSampleParameter2D_4, output=RGB}]}
  MaterialExpressionReroute_5 {class=MaterialExpressionReroute, inputs [None {from=MaterialExpressionNamedRerouteUsage_26}]}
  MaterialExpressionSaturate_0 {class=MaterialExpressionSaturate, inputs [None {from=MaterialExpressionMultiply_9}]}
  MaterialExpressionSaturate_1 {class=MaterialExpressionSaturate, inputs [None {from=MaterialExpressionMultiply_12}]}
  MaterialExpressionSaturate_2 {class=MaterialExpressionSaturate, inputs [None {from=MaterialExpressionMultiply_18}]}
  MaterialExpressionSaturate_3 {class=MaterialExpressionSaturate, inputs [None {from=MaterialExpressionSubtract_2}]}
  MaterialExpressionSaturate_4 {class=MaterialExpressionSaturate, inputs [None {from=MaterialExpressionSubtract_1}]}
  MaterialExpressionSaturate_5 {class=MaterialExpressionSaturate, inputs [None {from=MaterialExpressionMultiply_24}]}
  MaterialExpressionScalarParameter_0 {class=MaterialExpressionScalarParameter, properties {defaultValue=1, group=01_BaseTexture, parameterName=Contrast Base Texture}}
  MaterialExpressionScalarParameter_1 {
   class=MaterialExpressionScalarParameter
   properties {
    controlType=Enumeration
    defaultValue=1
    enumeration=/DealershipSharedLibrary/FX/Materials/Enums/Enum_DLR_TrueFalse
    group=01_BaseTexture
    parameterName=Use Diffuse Color Base Texture?
   }
  }
  MaterialExpressionScalarParameter_10 {class=MaterialExpressionScalarParameter, properties {group=000_General, parameterName=Portal Desaturation}}
  MaterialExpressionScalarParameter_11 {class=MaterialExpressionScalarParameter, properties {defaultValue=1, group=000_General, parameterName=Portal OutlineGlow}}
  MaterialExpressionScalarParameter_12 {class=MaterialExpressionScalarParameter, properties {defaultValue=1, group=07_WorldPosition Dissolve, parameterName=Dissolve Cat Radius In}}
  MaterialExpressionScalarParameter_13 {class=MaterialExpressionScalarParameter, properties {group=07_WorldPosition Dissolve, parameterName=Dissolve Cat Hardness In}}
  MaterialExpressionScalarParameter_14 {
   class=MaterialExpressionScalarParameter
   properties {defaultValue=1.4, group=07_WorldPosition Dissolve, parameterName=Dissolve Cat Contrast In}
  }
  MaterialExpressionScalarParameter_15 {class=MaterialExpressionScalarParameter, properties {group=07_WorldPosition Dissolve, parameterName=Dissolve Cat Radius Out}}
  MaterialExpressionScalarParameter_16 {
   class=MaterialExpressionScalarParameter
   properties {defaultValue=0.8, group=07_WorldPosition Dissolve, parameterName=Dissolve Cat Hardness Out}
  }
  MaterialExpressionScalarParameter_17 {
   class=MaterialExpressionScalarParameter
   properties {defaultValue=4, group=07_WorldPosition Dissolve, parameterName=Dissolve Cat Contrast Out}
  }
  MaterialExpressionScalarParameter_18 {class=MaterialExpressionScalarParameter, properties {defaultValue=-108, group=07_WorldPosition Dissolve, parameterName=Cat Dissolve Offset}}
  MaterialExpressionScalarParameter_19 {class=MaterialExpressionScalarParameter, properties {defaultValue=0.9, group=002_General, parameterName=Fresnel Contrast}}
  MaterialExpressionScalarParameter_2 {class=MaterialExpressionScalarParameter, properties {defaultValue=0.3, parameterName=Hardness SphreMask In}}
  MaterialExpressionScalarParameter_20 {class=MaterialExpressionScalarParameter, properties {defaultValue=5, group=002_General, parameterName=Fresnel Glow}}
  MaterialExpressionScalarParameter_21 {class=MaterialExpressionScalarParameter, properties {group=01_BaseTexture, parameterName=Use Screen Pos Uvs Base Texture}}
  MaterialExpressionScalarParameter_22 {class=MaterialExpressionScalarParameter, properties {group=02_Secondary Texture, parameterName=Use TextCoords? Second Texture}}
  MaterialExpressionScalarParameter_23 {class=MaterialExpressionScalarParameter, properties {group=03_Third Texture, parameterName=Use TextCoords? Third Texture}}
  MaterialExpressionScalarParameter_24 {class=MaterialExpressionScalarParameter, properties {group=05_Distorted Texture, parameterName=Use TextCoords? Distorted Texture}}
  MaterialExpressionScalarParameter_25 {class=MaterialExpressionScalarParameter, properties {defaultValue=1.3, group=002_General, parameterName=Final Adjustment}}
  MaterialExpressionScalarParameter_26 {
   class=MaterialExpressionScalarParameter
   properties {
    controlType=Enumeration
    enumeration=&Enum_DLR_TrueFalse
    group=002_General
    parameterName=Use Fresnel Method Color?
   }
  }
  MaterialExpressionScalarParameter_27 {class=MaterialExpressionScalarParameter, properties {defaultValue=1, group=01_BaseTexture, parameterName=Brightness Base Texture}}
  MaterialExpressionScalarParameter_28 {
   class=MaterialExpressionScalarParameter
   properties {defaultValue=10, group=02_Secondary Texture, parameterName=Brightness Result Secondary/Third Textures}
  }
  MaterialExpressionScalarParameter_29 {
   class=MaterialExpressionScalarParameter
   properties {defaultValue=10, group=02_Secondary Texture, parameterName=Brightness Adjustment Result Secondary/Third Textures}
  }
  MaterialExpressionScalarParameter_3 {class=MaterialExpressionScalarParameter, properties {defaultValue=0.3, parameterName=Hardness SphreMask 2 Out}}
  MaterialExpressionScalarParameter_30 {
   class=MaterialExpressionScalarParameter
   properties {
    controlType=Enumeration
    enumeration=&Enum_DLR_TrueFalse
    group=000_General
    parameterName=Use World Position Dissolve?
   }
  }
  MaterialExpressionScalarParameter_31 {class=MaterialExpressionScalarParameter, properties {defaultValue=3, group=002_General, parameterName=Fresnel Exponent}}
  MaterialExpressionScalarParameter_4 {
   class=MaterialExpressionScalarParameter
   properties {
    controlType=Enumeration
    defaultValue=1
    enumeration=&Enum_DLR_TrueFalse
    parameterName=Invert Sphere Mask 2?
   }
  }
  MaterialExpressionScalarParameter_5 {class=MaterialExpressionScalarParameter, properties {defaultValue=0.8, parameterName=Add SphereMaskXRadialTex}}
  MaterialExpressionScalarParameter_6 {
   class=MaterialExpressionScalarParameter
   properties {
    controlType=Enumeration
    enumeration=&Enum_DLR_TrueFalse
    group=000_General
    parameterName=UseDither?
   }
  }
  MaterialExpressionScalarParameter_7 {class=MaterialExpressionScalarParameter, properties {group=04_Distortion Texture, parameterName=Distortion Strength}}
  MaterialExpressionScalarParameter_8 {class=MaterialExpressionScalarParameter, properties {defaultValue=1, group=05_Distorted Texture, parameterName=Use RGB Distorted Texture?}}
  MaterialExpressionScalarParameter_9 {class=MaterialExpressionScalarParameter, properties {group=000_General, parameterName=Portal Desaturation}}
  MaterialExpressionScreenPosition_0 {class=MaterialExpressionScreenPosition, inputs [], properties {}}
  MaterialExpressionSphereMask_0 {
   class=MaterialExpressionSphereMask
   inputs [
    A {from=MaterialExpressionNamedRerouteUsage_4, output=""}
    B {from=MaterialExpressionConstant_0, output=""}
    Radius {from=MaterialExpressionNamedRerouteUsage_5, output=""}
    Hardness {from=MaterialExpressionScalarParameter_2, output=""}
   ]
   properties {}
  }
  MaterialExpressionSphereMask_1 {
   class=MaterialExpressionSphereMask
   inputs [
    A {from=MaterialExpressionNamedRerouteUsage_6, output=""}
    B {from=MaterialExpressionConstant_1, output=""}
    Radius {from=MaterialExpressionNamedRerouteUsage_8, output=""}
    Hardness {from=MaterialExpressionScalarParameter_3, output=""}
   ]
   properties {}
  }
  MaterialExpressionSphereMask_2 {
   class=MaterialExpressionSphereMask
   inputs [
    A {from=MaterialExpressionTextureCoordinate_2, output=""}
    B {from=MaterialExpressionNamedRerouteDeclaration_20, output=""}
    Radius {from=MaterialExpressionScalarParameter_12, output=""}
    Hardness {from=MaterialExpressionScalarParameter_13, output=""}
   ]
   properties {}
  }
  MaterialExpressionSphereMask_3 {
   class=MaterialExpressionSphereMask
   inputs [
    A {from=MaterialExpressionTextureCoordinate_3, output=""}
    B {from=MaterialExpressionNamedRerouteUsage_31, output=""}
    Radius {from=MaterialExpressionScalarParameter_15, output=""}
    Hardness {from=MaterialExpressionScalarParameter_16, output=""}
   ]
   properties {}
  }
  MaterialExpressionSubtract_0 {
   class=MaterialExpressionSubtract
   inputs [
    A {from=MaterialExpressionNamedRerouteDeclaration_22}
    B {from=MaterialExpressionNamedRerouteUsage_23}
   ]
  }
  MaterialExpressionSubtract_1 {
   class=MaterialExpressionSubtract
   inputs [
    A {from=MaterialExpressionSaturate_2}
    B {from=MaterialExpressionNamedRerouteUsage_24}
   ]
  }
  MaterialExpressionSubtract_2 {
   class=MaterialExpressionSubtract
   inputs [
    A {from=MaterialExpressionNamedRerouteUsage_25}
    B {from=MaterialExpressionNamedRerouteUsage_26}
   ]
  }
  MaterialExpressionSubtract_3 {
   class=MaterialExpressionSubtract
   inputs [
    A {from=MaterialExpressionWorldPosition_0, output=XYZ}
    B {from=MaterialExpressionObjectPositionWS_0}
   ]
  }
  MaterialExpressionSubtract_4 {
   class=MaterialExpressionSubtract
   inputs [
    A {from=MaterialExpressionObjectPositionWS_1}
    B {from=MaterialExpressionWorldPosition_2, output=XYZ}
   ]
  }
  MaterialExpressionTextureCoordinate_1 {class=MaterialExpressionTextureCoordinate}
  MaterialExpressionTextureCoordinate_2 {class=MaterialExpressionTextureCoordinate, properties {coordinateIndex=1}}
  MaterialExpressionTextureCoordinate_3 {class=MaterialExpressionTextureCoordinate, properties {coordinateIndex=1}}
  MaterialExpressionTextureSampleParameter2D_0 {
   class=MaterialExpressionTextureSampleParameter2D
   inputs [
    UVs {from=MaterialExpressionPanner_0, output=""}
    Apply View MipBias {from="", output=""}
   ]
   properties {group=01_BaseTexture, parameterName=Base Texture, texture=&/T_DLR_GalaxyCatPortalBaseB_D}
  }
  MaterialExpressionTextureSampleParameter2D_1 {
   class=MaterialExpressionTextureSampleParameter2D
   inputs [
    UVs {from=MaterialExpressionPanner_5, output=""}
    "DDX(UVs)" {from=MaterialExpressionDDX_0, output=""}
    "DDY(UVs)" {from=MaterialExpressionDDY_0, output=""}
    Apply View MipBias {from="", output=""}
   ]
   properties {
    group=06_Radial Opacity
    mipValueMode=TMVM_Derivative
    parameterName=Radial Texture Alpha
    texture=&/T_DLR_Dimensional_SmallParticles_A
   }
  }
  MaterialExpressionTextureSampleParameter2D_2 {
   class=MaterialExpressionTextureSampleParameter2D
   inputs [
    UVs {from=MaterialExpressionPanner_3, output=""}
    Apply View MipBias {from="", output=""}
   ]
   properties {
    group=02_Secondary Texture
    parameterName=Second Texture
    texture=&/T_DLR_BaseTexturesS23Cat_A
   }
  }
  MaterialExpressionTextureSampleParameter2D_3 {
   class=MaterialExpressionTextureSampleParameter2D
   inputs [
    UVs {from=MaterialExpressionPanner_4, output=""}
    Apply View MipBias {from="", output=""}
   ]
   properties {group=03_Third Texture, parameterName=Third Texture, texture=&T_DLR_BaseTexturesS23Cat_A}
  }
  MaterialExpressionTextureSampleParameter2D_4 {
   class=MaterialExpressionTextureSampleParameter2D
   inputs [
    UVs {from=MaterialExpressionPanner_6, output=""}
    Apply View MipBias {from="", output=""}
   ]
   properties {
    group=05_Distorted Texture
    parameterName=Distorted Texture
    texture=&/T_DLR_GalaxyCatPortal_D
   }
  }
  MaterialExpressionTextureSampleParameter2D_5 {
   class=MaterialExpressionTextureSampleParameter2D
   inputs [
    UVs {from=MaterialExpressionPanner_7, output=""}
    Apply View MipBias {from="", output=""}
   ]
   properties {group=04_Distortion Texture, parameterName=Distortion Texture, texture=/DealershipSharedLibrary/Assets/Sets/InkyOil/T_PerlinBlur}
  }
  MaterialExpressionTransform_1 {
   class=MaterialExpressionTransform
   inputs [None {from=MaterialExpressionCameraVectorWS_1}]
   properties {transformSourceType=TRANSFORMSOURCE_World, transformType=TRANSFORM_Tangent}
  }
  MaterialExpressionVectorParameter_0 {
   class=MaterialExpressionVectorParameter
   properties {
    group=01_BaseTexture
    parameterName="Tiling&&Offset BaseTexture"
    channelNames {a=Offset Y, b=Offset X, g=Tiling Y, r=Tiling X}
    defaultValue=<rgba 1 1 0 0>
   }
  }
  MaterialExpressionVectorParameter_10 {class=MaterialExpressionVectorParameter, properties {group=000_General, parameterName=Portal Outline Color, defaultValue=<rgba 1 0.2 0.1 1>}}
  MaterialExpressionVectorParameter_11 {
   class=MaterialExpressionVectorParameter
   properties {group=07_WorldPosition Dissolve, parameterName=Cat Dissolve Position Mask, channelNames {a="", b="", g=Y, r=X}}
  }
  MaterialExpressionVectorParameter_12 {class=MaterialExpressionVectorParameter, properties {group=002_General, parameterName=Fresnel Color, defaultValue=<rgba 0.141 0.454 1.1 0>}}
  MaterialExpressionVectorParameter_2 {
   class=MaterialExpressionVectorParameter
   properties {
    group=04_Distortion Texture
    parameterName="Tiling&&Offset Distortion Texture"
    channelNames {a=Offset Y, b=Offset X, g=Tiling Y, r=Tiling X}
    defaultValue=<rgba 1 1 0 0>
   }
  }
  MaterialExpressionVectorParameter_3 {
   class=MaterialExpressionVectorParameter
   properties {
    group=02_Secondary Texture
    parameterName="Tiling&&Offset Second Texture"
    channelNames {a=Offset Y, b=Offset X, g=Tiling Y, r=Tiling X}
    defaultValue=<rgba 1 1 0 0>
   }
  }
  MaterialExpressionVectorParameter_4 {
   class=MaterialExpressionVectorParameter
   properties {
    group=03_Third Texture
    parameterName="Tiling&&Offset Third Texture"
    channelNames {a=Offset Y, b=Offset X, g=Tiling Y, r=Tiling X}
    defaultValue=<rgba 1 1 0 0>
   }
  }
  MaterialExpressionVectorParameter_5 {
   class=MaterialExpressionVectorParameter
   properties {
    group=06_Radial Opacity
    parameterName=VectorToRadial Uvs
    channelNames {a=Offset Y, b=Offset X, g=Tiling Y, r=Tiling X}
    defaultValue=<rgba 2 2 0 0>
   }
  }
  MaterialExpressionVectorParameter_6 {
   class=MaterialExpressionVectorParameter
   properties {
    group=003_Speed Panners
    parameterName=Distortion And VectorToRadial Panners Speed
    channelNames {a=VectorToRadial Speed Y, b=VectorToRadial Speed X, g=Distortion Speed Y, r=Distortion Speed X}
    defaultValue=<rgba 0 0 0 -0.2>
   }
  }
  MaterialExpressionVectorParameter_7 {
   class=MaterialExpressionVectorParameter
   properties {
    group=05_Distorted Texture
    parameterName="Tiling&&Offset Distorted Texture"
    channelNames {a=Offset Y, b=Offset X, g=Tiling Y, r=Tiling X}
    defaultValue=<rgba 1 1 0 0>
   }
  }
  MaterialExpressionVectorParameter_8 {
   class=MaterialExpressionVectorParameter
   properties {
    group=003_Speed Panners
    parameterName="Panner Speed Offset Base && Second Texture"
    channelNames {a=Speed Y Second Texture, b=Speed X Second Texture, g=Speed Y Base Texture, r=Speed X Base Texture}
   }
  }
  MaterialExpressionVectorParameter_9 {
   class=MaterialExpressionVectorParameter
   properties {
    group=003_Speed Panners
    parameterName="Panner Speed Offset Third && Distorted Texture"
    channelNames {a=Speed Y Distorted Texture, b=Speed X Distorted Texture, g=Speed Y Third Texture, r=Speed X Third Texture}
   }
  }
  MaterialExpressionWorldPosition_0 {class=MaterialExpressionWorldPosition, inputs [], properties {}}
  MaterialExpressionWorldPosition_1 {class=MaterialExpressionWorldPosition, inputs [], properties {}}
  MaterialExpressionWorldPosition_2 {class=MaterialExpressionWorldPosition, inputs [], properties {}}
 }
 properties {
  bAutomaticallySetUsageInEditor=false
  bUsedWithNiagaraMeshParticles=true
  bUsedWithNiagaraSprites=true
  blendMode=BLEND_Masked
  shadingModel=MSM_Unlit
  usedShadingModels=Unlit
 }
}
